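# Material function for a SuperPrisma player cosmetic finish: anodized rainbow metal with animated camo, blending to gold via the Gold/GunMetal input.
# Fresnel plus Vertpan indexes a curve-atlas gradient, hue-shifted by the camo mask.
# Three panned texture projections on pre-skinned position, one sine-modulated, form that mask, which also varies brightness and roughness.
# Gmap and fresnel mask the emissive, thin film is optional, and mobile paths simplify the mask and wrapping.
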MaterialFunction /BRCosmetics/Characters/Player/Common/MaterialFunctions/SuperPrisma/MF_SL_AnodizedRainbow
graph {nodes [
  ComponentMask()
  n1 = FunctionInput() {inputName=norm}
  ComponentMask(None=@n1) {g=true}
  Constant3Vector() {constant=<rgba 0.224 0.2109 0.204 0>}
  Constant() {r=1}
  n2 = Constant() {r=1}
  n3 = ComponentMask(None=FunctionInput() {inputName=fresn}) {g=true}
  n4 = Add(A=Multiply(A=@n3) {constB=0.33}, B=FunctionInput() {bUsePreviewValueAsDefault=true, inputName=Vertpan, inputType=FunctionInput_Scalar})
  n5 = Frac(None=@n4)
  n6 = Abs(None=Sine(None=@n4))
  n7 = TextureObject() {samplerType=SAMPLERTYPE_LinearColor, texture=/Game/Characters/Player/Common/MaterialFunctions/Superlevel_DigitalCamo/SuperCamo_M}
  n8 = FunctionInput(Preview=VertexInterpolator(VS=PreSkinnedPosition()).PS) {bUsePreviewValueAsDefault=true, inputName=PreSkinnedLocal}
  n9 = ScalarParameter() {parameterName=Pan1_X}
  n10 = ScalarParameter() {parameterName=Pan1_Y}
  n11 = VertexInterpolator(VS=PreSkinnedNormal())
  n12 = Constant() {r=1}
  n13 = ComponentMask(None=MaterialFunctionCall!(
    TextureObject=@n7
    TextureSize=ScalarParameter() {defaultValue=18, parameterName=CamoScale}
    WorldPosition=@n8
    Export Float 4
    In=AppendVector(A=@n9, B=@n10)
    World Space Normal=@n11.PS
    ProjectionTransitionContrast=@n12
   ) {materialFunction=/Game/Athena/SafeZone/WorldAlignedTexture_ScalePan}.XYZ Texture) {r=true}
  n14 = ScalarParameter() {parameterName=Pan2_X}
  n15 = ScalarParameter() {defaultValue=-0.058, parameterName=Pan2_Y}
  n16 = ComponentMask(None=MaterialFunctionCall(
    TextureObject=@n7
    TextureSize=ScalarParameter() {defaultValue=20, parameterName=CamoScale3}
    WorldPosition=@n8
    In=AppendVector(A=@n14, B=@n15)
    World Space Normal=@n11.PS
    ProjectionTransitionContrast=@n12
   ) {materialFunction=&WorldAlignedTexture_ScalePan}.XYZ Texture) {g=true}
  n17 = ScalarParameter() {parameterName=Pan3_X}
  n18 = ScalarParameter() {defaultValue=0.05, parameterName=Pan3_Y}
  n19 = ComponentMask(None=MaterialFunctionCall(
    TextureObject=@n7
    TextureSize=ScalarParameter() {defaultValue=30, parameterName=CamoScale2}
    WorldPosition=@n8
    In=AppendVector(A=@n17, B=@n18)
    World Space Normal=@n11.PS
    ProjectionTransitionContrast=@n12
   ) {materialFunction=&WorldAlignedTexture_ScalePan}.XYZ Texture) {b=true}
  n20 = Sine(None=Multiply(A=ScalarParameter() {defaultValue=0.025, parameterName=Adjustment Speed}, B=Time() {period=0.5}))
  n21 = Add(A=Subtract(A=@n19), B=Clamp(None=Abs(None=@n20)) {maxDefault=0.9, minDefault=0.1})
  n22 = MaterialFunctionCall!(Default=Subtract(A=Max(A=@n13, B=@n16), B=Ceil(None=@n21)), Low/Mobile=Constant() {r=1}) {materialFunction=/Game/Effects/Fort_Effects/Materials/_Functions/MF_LowMobile_PathingSwitch}
  n23 = ScalarParameter() {defaultValue=0.05, parameterName=Camo Hue Boost}
  n24 = CurveAtlasRowParameter(CurveTime=Add(A=ShadingPathSwitch(Default=@n5, Mobile=@n6) {desc=Migrated from FeatureLevelSwitch}, B=Multiply(A=@n22.Result, B=@n23))) {
   atlas=&AthenaCurveAtlas01
   curve=/Game/Characters/Player/Common/MaterialFunctions/SuperPrisma/C_Color_Anodized_Prisma
   defaultValue=315
   parameterName=Curve
  }
  n25 = Clamp(None=Multiply(A=@n24, B=ScalarParameter() {defaultValue=1.1, parameterName=BrightnessBoost}))
  n26 = Multiply(A=@n25, B=ScalarParameter() {defaultValue=0.95, parameterName=CamoBoost})
  n27 = FunctionInput() {bUsePreviewValueAsDefault=true, inputName=Gold/GunMetal, inputType=FunctionInput_Scalar}
  n28 = ScalarParameter() {defaultValue=0.28, parameterName=Gold Roughness Min}
  n29 = ScalarParameter() {defaultValue=0.55, parameterName=Gold Roughness Max}
  n30 = Multiply(
   A=ScalarParameter() {defaultValue=0.2, parameterName=EmissiveIntensity}
   B=Multiply(A=OneMinus(None=@n27), B=FunctionInput() {inputName=Gmap})
  )
  n31 = SetMaterialAttributes!(
   MaterialAttributes
   BaseColor=LinearInterpolate(A=LinearInterpolate(A=@n25, B=@n26, Alpha=@n22.Result), B=Constant3Vector() {constant=<rgba 0.8229 0.471 0.0686 0>}, Alpha=@n27)
   Metallic=Constant() {r=1}
   Roughness=LinearInterpolate(A=LinearInterpolate(A=@n28, B=@n29, Alpha=@n22.Result), B=ScalarParameter() {defaultValue=0.2, parameterName=MetalRoughness}, Alpha=@n27)
   Anisotropy=ComponentMask(None=@n1) {g=true, r=true}
   EmissiveColor=Multiply(A=Multiply(A=@n24, B=@n30), B=@n3)
   Specular=Constant() {r=0.4}
   SubsurfaceColor=Constant()
  ) {attributeSetTypes [
    #1
    #2
    #3
    #4
    #5
    #6
    #7
   ]}
  FunctionOutput(None=StaticSwitchParameter(
    True=MaterialFunctionCall!(Alpha=@n2, Material Attributes=@n31) {materialFunction=/Game/Weapons/WeaponSkins/MaterialFunction/ThinFilmFunction}.Result
    False=@n31
   ) {defaultValue=true, parameterName=UseThinfilm})
  Round(None=@n19)
  ShaderStageSwitch()
  ShadingPathSwitch() {desc=Migrated from FeatureLevelSwitch}
 ]}
