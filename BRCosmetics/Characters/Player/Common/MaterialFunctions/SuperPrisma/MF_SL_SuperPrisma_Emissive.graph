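# Material function for SuperPrisma player cosmetics that sets a noise-modulated Fresnel glow as the emissive color.
# The Fresnel input is raised to an EmissiveFresnelPower parameter and multiplied by the HDR EmissiveColor parameter.
# The Noise input drives a lerp that scales this glow while keeping a low minimum.
# The result overwrites EmissiveColor on the incoming material attributes; three loose constants are unconnected.
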MaterialFunction /BRCosmetics/Characters/Player/Common/MaterialFunctions/SuperPrisma/MF_SL_SuperPrisma_Emissive
graph {nodes [
  Constant()
  Constant() {r=0.15}
  Constant() {r=0.8}
  n1 = Multiply(
   A=Power(
    Base=FunctionInput() {inputName=Fresnel, inputType=FunctionInput_Scalar}
    Exp=ScalarParameter() {defaultValue=4.5, parameterName=EmissiveFresnelPower}
   )
   B=VectorParameter() {parameterName=EmissiveColor, defaultValue=<rgba 50 31.7738 9.6934 1>}.RGB
  )
  n2 = LinearInterpolate(Alpha=FunctionInput() {inputName=Noise, inputType=FunctionInput_Scalar}) {constA=0.1}
  FunctionOutput(None=SetMaterialAttributes!(MaterialAttributes=FunctionInput() {inputType=FunctionInput_MaterialAttributes}, EmissiveColor=Multiply(A=@n1, B=@n2)) {attributeSetTypes [#1]})
 ]}
