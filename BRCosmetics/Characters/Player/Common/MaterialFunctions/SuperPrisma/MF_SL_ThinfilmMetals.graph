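# Material function for SuperPrisma player cosmetics that builds a gold or gunmetal metal surface with an optional thin-film effect.
# The Gold/GunMetal scalar input blends fixed base colors and the two roughness parameters, with full metallic and zero subsurface color.
# The static switch UseThinfilm, on by default, passes these material attributes and the same scalar through the weapon skins' ThinFilmFunction, or outputs them unchanged.
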MaterialFunction /BRCosmetics/Characters/Player/Common/MaterialFunctions/SuperPrisma/MF_SL_ThinfilmMetals
graph {nodes [
  n1 = FunctionInput() {bUsePreviewValueAsDefault=true, inputName=Gold/GunMetal, inputType=FunctionInput_Scalar}
  n2 = SetMaterialAttributes!(
   MaterialAttributes
   BaseColor=LinearInterpolate(A=Constant3Vector() {constant=<rgba 0.8229 0.471 0.0686 0>}, B=Constant3Vector() {constant=<rgba 0.224 0.2109 0.204 0>}, Alpha=@n1)
   Metallic=Constant() {r=1}
   Roughness=LinearInterpolate(
    A=ScalarParameter() {defaultValue=0.2, parameterName=Gold Roughness}
    B=ScalarParameter() {defaultValue=0.33, parameterName=MetalRoughness}
    Alpha=@n1
   )
   SubsurfaceColor=Constant()
  ) {attributeSetTypes [
    #1
    #2
    #3
    #4
   ]}
  FunctionOutput(None=StaticSwitchParameter(
    True=MaterialFunctionCall!(Alpha=@n1, Material Attributes=@n2) {materialFunction=/Game/Weapons/WeaponSkins/MaterialFunction/ThinFilmFunction}.Result
    False=@n2
   ) {defaultValue=true, parameterName=UseThinfilm})
 ]}
