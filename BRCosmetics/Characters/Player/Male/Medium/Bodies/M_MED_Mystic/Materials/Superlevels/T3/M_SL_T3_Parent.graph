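# Parent Cloth-shaded material for the Tier 3 Super Level skins of the Mystic character body.
# Two mask textures layer anodized rainbow, cloth, thin-film metal and rubber functions over the shared character PreFX attributes.
# A static switch adds toon shading, then masked SuperPrisma emissive uses panning world-aligned noise on pre-skinned position, with cheaper Low/Mobile inputs.
# Normals are recombined with the PreFX normal before the shared PostFX function.
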Material /BRCosmetics/Characters/Player/Male/Medium/Bodies/M_MED_Mystic/Materials/Superlevels/T3/M_SL_T3_Parent
graph {
 properties {
  bAutomaticallySetUsageInEditor=false
  bUseMaterialAttributes=true
  bUsedWithMeshDeformer=true
  bUsedWithMorphTargets=true
  bUsedWithSkeletalMesh=true
  pixelDepthOffsetMode=PDOM_Legacy
  shadingModel=MSM_Cloth
  usedShadingModels=Cloth
 }
 nodes [
  Constant()
  n1 = MaterialFunctionCall() {materialFunction=&BaseFN_CharacterFunction_PreFX}
  n2 = TextureSampleParameter2D() {
   parameterName=Mask02
   samplerSource=SSM_Wrap_WorldGroupSettings
   samplerType=SAMPLERTYPE_Masks
   texture=/BRCosmetics/Characters/Player/Male/Medium/Bodies/M_MED_Mystic/Textures/Skins/Super_Levels/T_Mystic_-_Super_Levels_Body_SL_FX_2
  }
  n3 = TextureSampleParameter2D() {
   parameterName=Mask01
   samplerSource=SSM_Wrap_WorldGroupSettings
   samplerType=SAMPLERTYPE_Masks
   texture=/BRCosmetics/Characters/Player/Male/Medium/Bodies/M_MED_Mystic/Textures/Skins/Super_Levels/T_Mystic_-_Super_Levels_Body_SL_FX
  }
  n4 = Max(A=Max(A=@n3.R, B=@n3.R), B=@n3.B)
  n5 = OneMinus(None=Max(A=@n2.B, B=Max(A=@n4, B=@n2.R)))
  n6 = GetMaterialAttributes(None=@n1.Result) {attributeGetTypes [#1]}
  n7 = ScalarParameter() {parameterName=Pan1_X}
  n8 = ScalarParameter() {defaultValue=0.05, parameterName=Pan1_Y}
  n9 = ComponentMask(None=MaterialFunctionCall!(
    TextureObject=TextureObject() {samplerType=SAMPLERTYPE_Masks, texture=/Game/Global/Textures/Default/Masks_and_Gradients/T_VertGradient_MASK}
    TextureSize=ScalarParameter() {defaultValue=550, parameterName=CamoScale}
    WorldPosition=@n1.Pre skinned Local Position
    Export Float 4
    In=AppendVector(A=@n7, B=@n8)
    World Space Normal=VertexNormalWS()
    ProjectionTransitionContrast=Constant() {r=1}
   ) {materialFunction=/Game/Athena/SafeZone/WorldAlignedTexture_ScalePan}.XY Texture) {g=true}
  n10 = Sine(None=Time()) {period=500}
  n11 = BlendMaterialAttributes(
   A=SetMaterialAttributes!(MaterialAttributes=@n6.MaterialAttributes, Normal=Constant3Vector() {constant=<rgba 0 0 1 0>}) {attributeSetTypes [#1]}
   B=MaterialFunctionCall!(
    Vertpan=MaterialFunctionCall!(Default=@n9, Low/Mobile=@n10) {materialFunction=/Game/Effects/Fort_Effects/Materials/_Functions/MF_LowMobile_PathingSwitch}.Result
    fresn=@n1.Gmap
    norm=@n6.Normal
    Gmap=@n1.Gmap
    PreSkinnedLocal=@n1.Gmap
    Gold/GunMetal=@n2.R
   ) {materialFunction=/BRCosmetics/Characters/Player/Common/MaterialFunctions/SuperPrisma/MF_SL_AnodizedRainbow}.Result
   Alpha=@n3.G
  )
  n12 = MaterialFunctionCall!(Black/White Mask=@n2.R) {materialFunction=/BRCosmetics/Characters/Player/Common/MaterialFunctions/SuperPrisma/MF_SL_SimpleClothLayers}
  n13 = OneMinus(None=@n2.R)
  n14 = BlendMaterialAttributes(
   A=BlendMaterialAttributes(
    A=BlendMaterialAttributes(A=@n11, B=@n12.Result, Alpha=@n3.R)
    B=MaterialFunctionCall!(Gold/GunMetal=@n13) {materialFunction=/BRCosmetics/Characters/Player/Common/MaterialFunctions/SuperPrisma/MF_SL_ThinfilmMetals}.Result
    Alpha=@n3.B
   )
   B=MaterialFunctionCall() {materialFunction=/BRCosmetics/Characters/Player/Common/MaterialFunctions/SuperPrisma/MF_SL_BasicRubber}.Result
   Alpha=@n2.B
  )
  n15 = StaticSwitchParameter(
   True=MaterialFunctionCall!(In=MaterialFunctionCall!(SkinMask=@n5, In=@n14) {materialFunction=/Game/Packages/Fortress_SharedMaterials/Anime_Master_Material/MF_Toonify}.Result) {materialFunction=/Game/Packages/Fortress_SharedMaterials/Anime_Master_Material/MF_AnimatedVisor}.Result
   False=@n14
  ) {parameterName=useAnimeShading}
  n16 = TextureObject() {samplerType=SAMPLERTYPE_LinearColor, texture=/Game/Weapons/WeaponSkins/Textures/T_Swirly_Noise_highContrast1}
  n17 = ScalarParameter() {defaultValue=466, parameterName=CamoScale2}
  n18 = Add(A=Add(A=@n1.Pre skinned Local Position, B=@n6.Normal), B=Multiply(A=@n1.Gmap) {constB=0.5})
  n19 = AppendVector(A=ScalarParameter() {parameterName=Pan1_X}, B=ScalarParameter() {defaultValue=0.03, parameterName=Pan1_Y2})
  n20 = VertexNormalWS()
  n21 = Constant() {r=1}
  n22 = MaterialFunctionCall(
   Default=ComponentMask(None=MaterialFunctionCall(TextureObject=@n16, TextureSize=@n17, WorldPosition=@n18, In=@n19, World Space Normal=@n20, ProjectionTransitionContrast=@n21) {materialFunction=&WorldAlignedTexture_ScalePan}.XY Texture) {g=true}
   Low/Mobile=Constant() {r=1}
  ) {materialFunction=&MF_LowMobile_PathingSwitch}
  n23 = Multiply(A=@n1.Gmap, B=@n1.Fresnel)
  n24 = GetMaterialAttributes(None=BlendMaterialAttributes(
    A=@n15
    B=MaterialFunctionCall!(Noise=@n22.Result, In=@n15, Fresnel=@n23) {materialFunction=/BRCosmetics/Characters/Player/Common/MaterialFunctions/SuperPrisma/MF_SL_SuperPrisma_Emissive}.Result
    Alpha=Multiply(A=@n1.Gmap, B=@n4)
   )) {attributeGetTypes [#1, #2, #3]}
  n25 = GetMaterialAttributes(None=@n1.Result) {attributeGetTypes [#1]}
  MaterialFunctionCall(
   AO=@n1.AO
   Distance Blend=@n1.AO
   Fresnel=@n1.AO
   Gmap=@n1.AO
   Pre Skinned Local Position=@n1.AO
   Material=SetMaterialAttributes!(
    MaterialAttributes=@n24.MaterialAttributes
    Normal=MaterialFunctionCall(BaseNormal=@n24.Normal, AdditionalNormal=@n25.Normal) {materialFunction=&BlendAngleCorrectedNormals}.Result
    Base Color=@n24.MaterialAttributes
    Metallic
   ) {attributeSetTypes [#1, #2, #4]}
  ) {materialFunction=&BaseFN_CharacterFunction_PostFX}
  ShadingModel()
 ]
}
